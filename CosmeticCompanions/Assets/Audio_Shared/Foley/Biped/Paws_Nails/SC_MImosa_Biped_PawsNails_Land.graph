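# Sound cue for the landing foley of clawed paws on biped cosmetic companions, playing a randomly picked paw-and-nail land sample.
# A quality-level switch routes its first two slots to a seven-sample random pool and its last two to a four-sample subset, trimming variation.
# A modulator at the root varies pitch and volume.
# Companion foley attenuation, the pets sound class and shared biped footstep concurrency govern playback.
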SoundCue /CosmeticCompanions/Assets/Audio_Shared/Foley/Biped/Paws_Nails/SC_MImosa_Biped_PawsNails_Land
root=Modulator_0
nodes [
 {id=Modulator_0, children [QualityLevel_0], properties {pitchMax=1.1, pitchMin=0.9, volumeMax=1, volumeMin=0.8}}
 {id=QualityLevel_0, children [Random_1,Random_1,Random_2,Random_2]}
 {
  id=Random_1
  children [WavePlayer_11, WavePlayer_12, WavePlayer_13, WavePlayer_14, WavePlayer_15, WavePlayer_16, WavePlayer_17]
  properties {weights [1,1,1,1,1,1,1]}
 }
 {id=Random_2, children [WavePlayer_18, WavePlayer_19, WavePlayer_20, WavePlayer_21], properties {weights [1,1,1,1]}}
 {
  id=WavePlayer_11
  properties {soundWaveAssetPtr=/CosmeticCompanions/Assets/Audio_Shared/Foley/Biped/Paws_Nails/Wavs/Mimosa_Biped_PawsNails_Land_01}
 }
 {
  id=WavePlayer_12
  properties {soundWaveAssetPtr=/CosmeticCompanions/Assets/Audio_Shared/Foley/Biped/Paws_Nails/Wavs/Mimosa_Biped_PawsNails_Land_02}
 }
 {
  id=WavePlayer_13
  properties {soundWaveAssetPtr=/CosmeticCompanions/Assets/Audio_Shared/Foley/Biped/Paws_Nails/Wavs/Mimosa_Biped_PawsNails_Land_03}
 }
 {
  id=WavePlayer_14
  properties {soundWaveAssetPtr=/CosmeticCompanions/Assets/Audio_Shared/Foley/Biped/Paws_Nails/Wavs/Mimosa_Biped_PawsNails_Land_04}
 }
 {
  id=WavePlayer_15
  properties {soundWaveAssetPtr=/CosmeticCompanions/Assets/Audio_Shared/Foley/Biped/Paws_Nails/Wavs/Mimosa_Biped_PawsNails_Land_05}
 }
 {
  id=WavePlayer_16
  properties {soundWaveAssetPtr=/CosmeticCompanions/Assets/Audio_Shared/Foley/Biped/Paws_Nails/Wavs/Mimosa_Biped_PawsNails_Land_06}
 }
 {
  id=WavePlayer_17
  properties {soundWaveAssetPtr=/CosmeticCompanions/Assets/Audio_Shared/Foley/Biped/Paws_Nails/Wavs/Mimosa_Biped_PawsNails_Land_07}
 }
 {
  id=WavePlayer_18
  properties {soundWaveAssetPtr=&Mimosa_Biped_PawsNails_Land_01}
 }
 {
  id=WavePlayer_19
  properties {soundWaveAssetPtr=&Mimosa_Biped_PawsNails_Land_02}
 }
 {
  id=WavePlayer_20
  properties {soundWaveAssetPtr=&Mimosa_Biped_PawsNails_Land_03}
 }
 {
  id=WavePlayer_21
  properties {soundWaveAssetPtr=&Mimosa_Biped_PawsNails_Land_04}
 }
]
properties {
 attenuationSettings=/Game/Sounds/Attenuation/AS_Companion_Foley
 maxDistance=1200
 pitchMultiplier=0.9
 priority=0
 soundClassObject=&Fort_Creatures_Pets
 volumeMultiplier=0.7
 concurrencySet [/CosmeticCompanions/Assets/Audio_Shared/Foley/Biped/CONC_Mimosa_Biped_Footstep]
}
